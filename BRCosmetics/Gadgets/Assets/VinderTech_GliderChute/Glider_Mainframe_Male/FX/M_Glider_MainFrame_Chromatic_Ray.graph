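# Unlit, two-sided translucent material for Niagara mesh particles drawing chromatic rays on the Mainframe glider.
# Noise and cube-pattern textures distort two oppositely offset turbulence samples, each tinted differently for a chromatic split, then colored by a curve-atlas gradient and particle color.
# Opacity combines UV and smoke masks, view-angle fade, a particle-relative mesh mask and time-of-day brightness.
# Dynamic parameters drive UV, color, panning, masking and noise-based normal displacement.
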Material /BRCosmetics/Gadgets/Assets/VinderTech_GliderChute/Glider_Mainframe_Male/FX/M_Glider_MainFrame_Chromatic_Ray
graph {
 properties {
  bUsedWithNiagaraMeshParticles=true
  blendMode=BLEND_Translucent
  pixelDepthOffsetMode=PDOM_Legacy
  shadingModel=MSM_Unlit
  twoSided=true
  usedShadingModels=Unlit
 }
 nodes [
  n1 = ParticleColor()
  n2 = TextureCoordinate()
  n3 = DynamicParameter() {defaultValue=<rgba 0 0 1 1>, paramNames [UV_offset,Color_offset,WPO,Panner]}
  n4 = CurveAtlasRowParameter(CurveTime=TextureSample(UVs=Add(A=@n2, B=@n3.Color_offset) {constB=0}) {texture=/Game/Global/Textures/Default/Masks_and_Gradients/WeaponGradients/T_Gradient_DesertSky_Color}.R) {
   atlas=&AthenaCurveAtlas01
   curve=/Game/Athena/Prototype/Blueprints/Cattus/FX/Curve_Color2
   defaultValue=113
   parameterName=Gradient color
  }
  n5 = Constant() {r=-0.5}
  n6 = VectorParameter() {parameterName=Disturb_UV_Scale, defaultValue=<rgba 6 0.7 0 1>}
  n7 = Add(
   A=DynamicParameter() {defaultValue=<rgba 0 0 1 1>, paramNames [UV_offset,Color_offset,WPO,Panner]}.UV_offset
   B=Multiply(A=TextureCoordinate(), B=AppendVector(A=@n6.R, B=@n6.R))
  )
  n8 = TextureSample(UVs=Panner(Coordinate=@n7) {speedY=0.5}) {samplerType=SAMPLERTYPE_LinearColor, texture=&T_NoisePack_LinearColor}
  n9 = TextureCoordinate() {uTiling=6, vTiling=2}
  n10 = ComponentMask(None=TextureCoordinate()) {g=true}
  n11 = LinearInterpolate(
   A=@n8.G
   B=TextureSample(UVs=Panner(Coordinate=@n9) {speedY=0.5}) {samplerType=SAMPLERTYPE_LinearColor, texture=/Game/UI/Foundation/Textures/Pattern/T_CubePattern}.G
   Alpha=Saturate(None=Power(Base=@n10) {constExponent=1})
  )
  n12 = DynamicParameter() {parameterIndex=1, defaultValue=<rgba 0.5 0.1 1 1>, paramNames [Mesh_mask, Square_Texture_lerp, Param3, Param4]}
  n13 = Panner(Coordinate=@n7) {speedX=0.2, speedY=-0.2}
  n14 = Multiply(
   A=AppendVector(
    A=LinearInterpolate(A=@n11, B=@n8.G, Alpha=@n12.Square_Texture_lerp) {constAlpha=0.1}
    B=TextureSample(UVs=@n13) {samplerType=SAMPLERTYPE_LinearColor, texture=&T_NoisePack_LinearColor}.R
   )
   B=ScalarParameter() {defaultValue=0.15, parameterName=Disturb_Intensity}
  ) {constB=0.15}
  n15 = VectorParameter() {parameterName=Noise_UV_Scale, defaultValue=<rgba 3 0.5 0 1>}
  n16 = Multiply(A=TextureCoordinate(), B=AppendVector(A=@n15.R, B=@n15.R))
  n17 = Constant() {r=0.008}
  n18 = DynamicParameter() {defaultValue=<rgba 0 0 1 1>, paramNames [UV_offset,Color_offset,WPO,Panner]}
  n19 = Add(A=Add(A=@n14, B=Add(A=@n16, B=@n17) {constB=0.05}), B=@n18.UV_offset)
  n20 = Multiply(
   A=VectorParameter() {parameterName=Chroma3, defaultValue=<rgba 1 0.0477 0.5272 1>}.RGB
   B=TextureSample(UVs=Panner(Coordinate=@n19, Time=@n18.Panner) {speedY=0.3}) {texture=/Game/Global/Textures/Default/Caustics/T_Turbulence_Seamless}.R
  )
  n21 = Add(A=@n18.UV_offset, B=Add(A=Subtract(A=@n16, B=@n17), B=@n14))
  n22 = Multiply(
   A=TextureSample(UVs=Panner(Coordinate=@n21, Time=@n18.Panner) {speedY=0.3}) {texture=&T_Turbulence_Seamless}.R
   B=VectorParameter() {parameterName=Chroma1, defaultValue=<rgba 0.0296 1.3506 2 1>}.RGB
  )
  n23 = Power(Base=ComponentMask(None=TextureCoordinate()) {g=true}) {constExponent=6}
  n24 = Power(
   Base=Multiply(A=LinearInterpolate(A=@n20, B=@n22)) {constB=4}
   Exp=Power(Base=LinearInterpolate(Alpha=@n23) {constA=1, constB=0}) {constExponent=5}
  ) {constExponent=1}
  n25 = Multiply(A=@n1.RGB, B=Multiply(A=Desaturation(None=@n4, Fraction=@n5), B=@n24))
  MaterialFunctionCall(Emissive=@n25) {materialFunction=&MF_FX_EmissiveAdjustments}
  n26 = ParticlePositionWS()
  n27 = WorldPosition()
  n28 = Transform(None=Divide(A=Subtract(A=@n26, B=@n27.XYZ)) {constB=1024}) {transformSourceType=TRANSFORMSOURCE_World, transformType=TRANSFORM_Local}
  n29 = DynamicParameter() {parameterIndex=1, defaultValue=<rgba 0.5 0.1 1 1>, paramNames [Mesh_mask, Square_Texture_lerp, Param3, Param4]}
  n30 = Multiply(A=Add(A=ComponentMask(None=@n28) {g=true}, B=LinearInterpolate(Alpha=@n29.Mesh_mask) {constA=-0.2, constB=0.2})) {constB=20}
  n31 = Abs(None=MaterialFunctionCall("FadeSensitivity (optional)"=ScalarParameter() {parameterName=FadeBasedOnViewAngle}) {materialFunction=&FadeBasedOnViewAngle}.Result)
  n32 = ComponentMask(None=TextureCoordinate()) {g=true}
  n33 = Power(Base=@n32) {constExponent=3}
  n34 = OneMinus(None=@n32)
  n35 = Saturate(None=Multiply(A=Power(Base=@n34) {constExponent=1}) {constB=5})
  n36 = Add(A=TextureCoordinate(), B=Constant() {r=0.05})
  n37 = Multiply(
   A=@n24
   B=Multiply(
    A=Multiply(A=@n33, B=@n35)
    B=TextureSample(UVs=@n36) {samplerType=SAMPLERTYPE_Grayscale, texture=/Game/Effects/Fort_Effects/Textures/Smoke/T_Smoke_Trail_Soft_NoClamp}.R
   )
  )
  n38 = Multiply(A=Multiply(A=@n1.A, B=Multiply(A=@n37)))
  Multiply(
   A=Multiply(A=Multiply(A=@n38, B=@n31), B=Saturate(None=@n30))
   B=MaterialFunctionCall(
    Color_Morning=ScalarParameter() {defaultValue=1, group=Color_time_of_day, parameterName=Morning_brightness}
    Color_Day=ScalarParameter() {defaultValue=1, group=Color_time_of_day, parameterName=Day_brightness}
    Color_Evening=ScalarParameter() {defaultValue=1, group=Color_time_of_day, parameterName=Evening_brightness}
    Color_Night=ScalarParameter() {defaultValue=1, group=Color_time_of_day, parameterName=Night_brightness}
   ) {materialFunction=&Color_Time_of_Day}.Result
  )
  Reroute(None=@n25)
  n39 = VertexNormalWS()
  n40 = TextureCoordinate() {uTiling=6, vTiling=0.7}
  n41 = ComponentMask(None=TextureSample(UVs=Panner(Coordinate=@n40) {speedY=0.5}) {samplerType=SAMPLERTYPE_LinearColor, texture=&T_NoisePack_LinearColor}.RGB) {b=true, g=true, r=true}
  Reroute(None=Multiply(A=DynamicParameter() {defaultValue=<rgba 0 0 1 1>, paramNames [UV_offset,Color_offset,WPO,Panner]}.WPO, B=Multiply(A=@n39, B=@n41)))
 ]
}
