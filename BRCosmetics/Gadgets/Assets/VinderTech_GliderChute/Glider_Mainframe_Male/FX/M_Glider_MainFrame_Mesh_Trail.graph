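# Additive unlit, two-sided material for the Mainframe glider's trail, built for Niagara mesh particles and ribbons.
# Niagara dynamic parameters drive a clamped world-position bend weighted along U, with optional SplineThicken.
# Param1-shaped UV sphere masks modulate tinted particle color and opacity, which also takes depth, near-camera and view-angle fades.
# A switch substitutes plain particle color with a local-height wipe driven by Disappear; time-of-day brightness scales the final mask.
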Material /BRCosmetics/Gadgets/Assets/VinderTech_GliderChute/Glider_Mainframe_Male/FX/M_Glider_MainFrame_Mesh_Trail
graph {
 properties {
  bAutomaticallySetUsageInEditor=false
  bEnableResponsiveAA=true
  bUsedWithNiagaraMeshParticles=true
  bUsedWithNiagaraRibbons=true
  bUsedWithStaticMesh=false
  blendMode=BLEND_Additive
  pixelDepthOffsetMode=PDOM_Legacy
  shadingModel=MSM_Unlit
  twoSided=true
  usedShadingModels=Unlit
 }
 nodes [
  n1 = DynamicParameter() {desc=Parameters set by Niagara, parameterIndex=1, paramNames [X,Y,Z,WPOIntensity]}
  n2 = Subtract(
   A=MaterialFunctionCall(X=@n1.X, Y=@n1.X, Z=@n1.X) {materialFunction=&MakeFloat3}.Result
   B=ObjectPositionWS()
  )
  n3 = ScalarParameter() {defaultValue=-300, parameterName=Clamp_min}
  n4 = ScalarParameter() {defaultValue=300, parameterName=Clamp_max}
  n5 = TextureCoordinate()
  n6 = Multiply(A=Power(Base=ComponentMask(None=@n5) {r=true}, Exp=ScalarParameter() {defaultValue=3, parameterName=WPOBendPower}), B=@n1.WPOIntensity)
  n7 = LinearInterpolate(
   A=Constant()
   B=LinearInterpolate(B=Clamp(None=@n2, Min=@n3, Max=@n4) {maxDefault=1000, minDefault=-1000}, Alpha=LinearInterpolate(Alpha=@n6) {constB=0.8})
   Alpha=ScalarParameter() {defaultValue=1, parameterName=Use interpolate WPO}
  )
  n8 = Multiply(A=Constant() {r=1}, B=DynamicParameter() {paramNames [Param1, Param2, Disappear, SplineThicken]}.SplineThicken)
  LinearInterpolate(
   A=@n7
   B=Add(
    A=@n7
    B=MaterialFunctionCall(WidthBase=@n8, WidthTip=@n8) {materialFunction=&SplineThicken}.WorldPositionOffset
   )
   Alpha=ScalarParameter() {parameterName=Use_SplineThicken}
  )
  n9 = ComponentMask(None=TextureCoordinate()) {r=true}
  n10 = ParticleColor()
  n11 = TextureCoordinate()
  n12 = ComponentMask(None=@n11) {r=true}
  n13 = Constant() {r=0.5}
  n14 = ComponentMask(None=@n11) {g=true}
  n15 = DynamicParameter() {paramNames [Param1, Param2, Disappear, SplineThicken]}
  n16 = SphereMask(
   A=@n14
   B=@n13
   Radius=LinearInterpolate(Alpha=@n15.Param1) {constA=0.01, constB=0.5}
   Hardness=LinearInterpolate(Alpha=@n15.Param1) {constB=0.9}
  ) {hardnessPercent=70}
  n17 = Constant() {r=0.5}
  n18 = SphereMask(A=@n14, B=@n17, Radius=@n17) {attenuationRadius=0.4, hardnessPercent=0}
  n19 = LinearInterpolate(
   A=LinearInterpolate(Alpha=Multiply(A=@n15.Param1, B=@n15.Param1)) {constA=1, constB=0.01}
   Alpha=Power(Base=OneMinus(None=@n18)) {constExponent=2.5}
  ) {constA=0.04}
  n20 = Add(A=@n19) {constB=0.1}
  n21 = Multiply(A=@n10.A, B=Multiply(A=SphereMask(A=@n12, B=@n13, Radius=@n13) {hardnessPercent=90}, B=Multiply(A=@n16, B=@n20)))
  n22 = Multiply(
   A=LinearInterpolate(B=VectorParameter() {parameterName=DarkColor, defaultValue=<rgba 0.1718 0.2044 0.5 0>}.RGB, Alpha=Multiply(A=@n9, B=@n21)) {constA=1}
   B=@n10.RGB
  )
  n23 = ScalarParameter() {parameterName=Only_Use_WorldPosition_Disappear}
  n24 = ParticleColor()
  MaterialFunctionCall(Emissive=LinearInterpolate(A=Multiply(A=@n22, B=@n19), B=@n24.RGB, Alpha=@n23)) {materialFunction=&MF_FX_EmissiveAdjustments}
  n25 = DepthFade()
  n26 = Constant() {r=1}
  n27 = ShadingPathSwitch(Default=@n25, Deferred=@n25, Forward=@n26, Mobile=@n26)
  n28 = ComponentMask(None=TextureCoordinate()) {r=true}
  n29 = Multiply(
   A=Power(Base=OneMinus(None=@n28), Exp=ScalarParameter() {defaultValue=1.5, parameterName=Mask Power}) {constExponent=1.5}
   B=ScalarParameter() {defaultValue=5, parameterName=Mask Multiply}
  ) {constB=5}
  n30 = Multiply(
   A=Multiply(A=Multiply(A=@n21, B=@n27), B=Saturate(None=@n29))
   B=MaterialFunctionCall(Distance=ScalarParameter() {defaultValue=100, parameterName=Near_camera_fade}) {materialFunction=&Near_Camera_Fade}.Result
  )
  n31 = MaterialFunctionCall("FadeSensitivity (optional)"=ScalarParameter() {parameterName=FadeBasedOnViewAngel}) {materialFunction=&FadeBasedOnViewAngle}
  n32 = Subtract(A=ParticlePositionWS(), B=WorldPosition().XYZ)
  n33 = ComponentMask(None=Transform(None=Divide(A=@n32) {constB=1024}) {transformSourceType=TRANSFORMSOURCE_World, transformType=TRANSFORM_Local}) {g=true}
  n34 = LinearInterpolate(Alpha=DynamicParameter() {paramNames [Param1, Param2, Disappear, SplineThicken]}.Disappear) {constA=-0.2, constB=0.2}
  n35 = Saturate(None=Multiply(A=Add(A=@n33, B=@n34)) {constB=20})
  Multiply(
   A=LinearInterpolate(A=Multiply(A=@n30, B=@n31.Result), B=Multiply(A=@n24.A, B=@n35), Alpha=@n23)
   B=MaterialFunctionCall(
    Color_Morning=ScalarParameter() {defaultValue=1, group=Color_time_of_day, parameterName=Morning_brightness}
    Color_Day=ScalarParameter() {defaultValue=1, group=Color_time_of_day, parameterName=Day_brightness}
    Color_Evening=ScalarParameter() {defaultValue=1, group=Color_time_of_day, parameterName=Evening_brightness}
    Color_Night=ScalarParameter() {defaultValue=1, group=Color_time_of_day, parameterName=Night_brightness}
   ) {materialFunction=&Color_Time_of_Day}.Result
  )
 ]
}
