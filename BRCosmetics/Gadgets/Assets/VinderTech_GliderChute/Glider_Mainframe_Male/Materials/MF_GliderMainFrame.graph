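# Material function for the Mainframe Male glider, outputting PBR material attributes with optional clear coat metallic, supercharged glow and back-face fresnel emissive.
# A FadeIn scalar, or looping test time, drives a two-stage materialize effect: glowing curvature edges from normal and position derivatives, then the full surface.
# Sphere masks over UV and pre-skinned height build the opacity reveal.
# Optional ShadowReplace offsets collapse shadows until the main surface starts appearing.
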MaterialFunction /BRCosmetics/Gadgets/Assets/VinderTech_GliderChute/Glider_Mainframe_Male/Materials/MF_GliderMainFrame
graph {nodes [
  Constant() {r=1}
  n1 = FunctionInput(Preview=TextureCoordinate()) {bUsePreviewValueAsDefault=true, inputName=UV 1 coordinates, inputType=FunctionInput_Vector2, sortPriority=2}
  n2 = TextureSampleParameter2D(UVs=@n1) {
   group=Diffuse
   parameterName=Diffuse
   samplerSource=SSM_Wrap_WorldGroupSettings
   texture=/BRCosmetics/Gadgets/Assets/VinderTech_GliderChute/Glider_Mainframe_Male/Textures/T_Glider_Mainframe_Male_D
  }
  n3 = TextureSampleParameter2D(UVs=@n1) {
   desc="R = SpecMask G = Metallic B = Roughness"
   group=Specular
   parameterName=SpecularMasks
   samplerSource=SSM_Wrap_WorldGroupSettings
   samplerType=SAMPLERTYPE_Masks
   texture=/BRCosmetics/Gadgets/Assets/VinderTech_GliderChute/Glider_Mainframe_Male/Textures/T_Glider_Mainframe_Male_S
  }
  n4 = LinearInterpolate(A=@n3.G, B=Multiply(A=@n3.A, B=ScalarParameter() {defaultValue=1, group=ClearCoat, parameterName=PaintMetallic}), Alpha=@n3.G)
  n5 = ScalarParameter() {defaultValue=1, group=Specular, parameterName=SpecularLevel}
  n6 = ScalarParameter() {group=Roughness, parameterName=SpecRoughnessMin}
  n7 = ScalarParameter() {defaultValue=1, group=Roughness, parameterName=SpecRoughnessMax}
  n8 = TextureSampleParameter2D(UVs=@n1) {
   group=Normals
   parameterName=Normals
   samplerType=SAMPLERTYPE_Normal
   texture=/BRCosmetics/Gadgets/Assets/VinderTech_GliderChute/Glider_Mainframe_Male/Textures/T_Glider_Mainframe_Male_N
  }
  n9 = StaticSwitchParameter(
   True=MaterialFunctionCall!(WPO) {materialFunction=/Game/Packages/Fortress_SharedMaterials/Base_Material_Functions/Fortnite_HideElementsWithVertexColor}.Result
   False=Constant()
  ) {group=Opacity Mask, parameterName=Use Vertex Colors for Mask}
  n10 = NamedRerouteUsage()
  n11 = TextureObjectParameter() {
   group=Emissive
   parameterName=EmissiveTexture
   samplerType=SAMPLERTYPE_Masks
   texture=/BRCosmetics/Gadgets/Assets/VinderTech_GliderChute/Glider_Mainframe_Male/Textures/T_Glider_Mainframe_Male_E
  }
  n12 = Multiply(
   A=Add(A=TextureSample(Tex=@n11).RGB, B=NamedRerouteUsage())
   B=VectorParameter() {group=Emissive, parameterName=EmissiveMultiplier, defaultValue=<rgba 1 1 1 1>}.RGB
  )
  n13 = MaterialFunctionCall!(
   AverageHandPosition
   SuperChargedFlickerBias=ScalarParameter() {defaultValue=1, group=Super Charge, parameterName=SuperChargedFlickerBias}
   SuperChargedEffect=ScalarParameter() {group=Super Charge, parameterName=SuperChargedEffect}
   UseDistanceCalc=StaticBool()
   Emissive
  ) {materialFunction=/Game/Characters/Fortress_CharacterShader/SuperChargedGlow}
  n14 = Clamp(None=Multiply(A=TwoSidedSign()) {constB=-1})
  n15 = VectorParameter() {parameterName=colcol1, defaultValue=<rgba 6.5583 9.5135 20 0.2>}
  n16 = StaticBool()
  n17 = LinearInterpolate(
   A=@n15.RGB
   B=VectorParameter() {parameterName=colcol2, defaultValue=<rgba 1.1035 1.5954 6 1>}.RGB
   Alpha=Abs(None=MaterialFunctionCall(Power=@n15.A, Clamp Fresnel Dot Product=@n16) {materialFunction=&Fresnel_Function}.Result)
  )
  n18 = ShadingPathSwitch(Default=Add(A=Add(A=@n12, B=@n13.Result), B=Multiply(A=@n14, B=@n17)), Mobile=@n12) {desc=Migrated from FeatureLevelSwitch}
  n19 = MaterialFunctionCall() {materialFunction=/Game/Packages/Fortnite_UniqueMaterials_1/Weapons/WeaponLocalPosition}
  FunctionOutput(None=MakeMaterialAttributes!(
    BaseColor=@n2.RGB
    Metallic=StaticSwitchParameter(True=@n4, False=@n3.G) {parameterName=isClearCoat}
    Specular=Multiply(A=@n3.R, B=@n5)
    Roughness=LinearInterpolate(A=@n6, B=@n7, Alpha=@n3.B)
    Anisotropy
    EmissiveColor=QualitySwitch(Default=@n18, Low=@n12, High=@n18, Medium=@n12)
    Opacity
    OpacityMask=NamedRerouteUsage()
    Normal=@n8.RGB
    Tangent
    WorldPositionOffset=Add(A=@n9, B=@n10) {constB=0}
    SubsurfaceColor=Constant3Vector()
    ClearCoat=@n3.A
    ClearCoatRoughness=ScalarParameter() {group=ClearCoat, parameterName=ClearCoatRoughness}
    AmbientOcclusion
    Refraction
    CustomizedUVs_0
    CustomizedUVs_1
    CustomizedUVs_2=@n19.Custom Tex Coord 2
    CustomizedUVs_3=@n19.Custom Tex Coord 3
    CustomizedUVs_4
    CustomizedUVs_5
    CustomizedUVs_6
    CustomizedUVs_7
    PixelDepthOffset
    ShadingModel
    Displacement
   )) {outputName=Result Material Attributes}
  Multiply(
   A=@n2.A
   B=MaterialFunctionCall!(Animation Gradient, Rez In Percentage) {materialFunction=/Game/Packages/Fortress_SharedMaterials/Base_Material_Functions/CharacterResInEffect}.Modulative Opacity
  )
  n20 = PixelNormalWS()
  n21 = Add(A=Abs(None=DDX(Value=@n20)), B=Abs(None=DDY(Value=@n20)))
  n22 = Constant()
  n23 = WorldPosition()
  n24 = Add(A=Abs(None=DDX(Value=@n23.XYZ)), B=Abs(None=DDY(Value=@n23.XYZ)))
  n25 = Constant()
  n26 = Multiply(A=Constant() {r=1}, B=ComponentMask(None=TextureCoordinate()) {r=true})
  n27 = Add(A=PreSkinnedPosition(), B=Constant() {r=150})
  n28 = Constant() {r=-0.005}
  n29 = VertexInterpolator(VS=ComponentMask(None=Multiply(A=@n27, B=@n28)) {g=true})
  n30 = Multiply(A=Time(), B=ScalarParameter() {defaultValue=0.5, group=MainframeEffect, parameterName=Speed})
  n31 = Saturate(None=StaticSwitchParameter(True=Frac(None=@n30), False=ScalarParameter() {group=MainframeEffectTiming, parameterName=FadeIn, sliderMax=1, sortPriority=0}) {group=MainframeEffect, parameterName=TestAnimation})
  n32 = MaterialFunctionCall(
   Input=@n31
   Input Low=ScalarParameter() {group=MainframeEffectTiming, parameterName=OutlineFadeInStart, sliderMax=1, sortPriority=1}
   Input High=ScalarParameter() {defaultValue=0.7, group=MainframeEffectTiming, parameterName=OutlineFadeInEnd, sliderMax=1, sortPriority=2}
   Target Low=Constant()
   Target High=Constant() {r=1}
  ) {materialFunction=&RemapValueRange}
  n33 = LinearInterpolate(
   B=Divide(A=Distance(A=@n21, B=@n22), B=Distance(A=@n24, B=@n25))
   Alpha=SphereMask(A=Add(A=@n26, B=@n29.PS), B=Constant(), Radius=Saturate(None=@n32.Result), Hardness=Constant() {r=1})
  )
  n34 = TextureCoordinate()
  n35 = PreSkinnedPosition()
  n36 = Constant() {r=300}
  n37 = ComponentMask(None=Multiply(A=Add(A=@n35, B=@n36), B=Constant() {r=-0.002})) {g=true}
  n38 = Add(A=Multiply(A=ComponentMask(None=@n34) {r=true}, B=Constant() {r=0.7}), B=VertexInterpolator(VS=@n37).PS)
  n39 = Constant()
  n40 = Saturate(None=MaterialFunctionCall(
    Input=@n31
    Input Low=ScalarParameter() {defaultValue=0.5, group=MainframeEffectTiming, parameterName=MainMaterialFadeInStart, sliderMax=1, sortPriority=3}
    Input High=ScalarParameter() {defaultValue=1, group=MainframeEffectTiming, parameterName=MainMaterialFadeInEnd, sliderMax=1, sortPriority=4}
    Target Low=Constant()
    Target High=Constant() {r=1}
   ) {materialFunction=&RemapValueRange}.Result)
  n41 = Constant() {r=0.7}
  NamedRerouteDeclaration(None=Max(A=@n33, B=SphereMask(A=@n38, B=@n39, Radius=@n40, Hardness=@n41))) {name=OutlineOpacity, nodeColor=<rgba 0.9688 1 0 1>}
  n42 = OneMinus(None=@n40)
  NamedRerouteDeclaration(None=Multiply(
    A=Multiply(A=@n33, B=@n42)
    B=VectorParameter() {group=MainframeEffect, parameterName=MainFrameEmissive, defaultValue=<rgba 0 0.3572 5 0>}.RGB
   )) {name=OutlineEmissive, nodeColor=<rgba 1 0 0.8672 1>}
  n43 = Constant()
  n44 = StaticSwitchParameter(True=Subtract(A=ObjectPositionWS(), B=WorldPosition().XYZ), False=@n43) {defaultValue=true, group=MainframeEffect, parameterName=DisableShadows}
  NamedRerouteDeclaration(None=LinearInterpolate(A=ShadowReplace(Default=@n43, Shadow=@n44), Alpha=Ceil(None=@n40)) {constB=0}) {name=OutlineWPO, nodeColor=<rgba 1 0 0.1641 1>}
  OneMinus(None=@n37)
  VertexNormalWS()
 ]}
